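# Widget Blueprint for one kill feed row in the Athena spectator broadcast HUD, parented to AthenaBroadcastKillFeedItemBase.
# A horizontal row shows a dark instigator name plate, an event icon and a gradient, blurred victim name plate, with collapsed elimination icons beside both names.
# The event graph is empty; name texts and icon are inherited bindings.
# SlideFadeIn animates the row, while SlideFadeOut and SlideUp target a RetainerBox absent from the widget tree.
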WidgetBlueprint /FortniteCoreUI/Athena/HUD/Spectator/KillFeed/AthenaBroadcastKillFeedItem
animations [
 SlideFadeIn {
  binding_count=1
  length_seconds=0.25
  object=&self:SlideFadeIn
  sequence {bindings [HorizontalBox_0 {
     id=#1
     display_name=SizeBox_0
     possessed_class=HorizontalBox
     tracks [
      {
       type=MovieSceneFloatTrack
       display_name=Render Opacity
       sections [{type=MovieSceneFloatSection, end_frame=5, start_frame=0, channels [None {type=MovieSceneScriptingFloatChannel, keys [[0,0],[5,1]]}]}]
      }
      {
       type=MovieScene2DTransformTrack
       sections [{
         type=MovieScene2DTransformSection
         end_frame=5
         start_frame=0
         channels [
          Translation.X {type=MovieSceneScriptingFloatChannel, const [0,1]}
          Translation.Y {type=MovieSceneScriptingFloatChannel, keys [[0,10],[5,0]]}
          Angle {type=MovieSceneScriptingFloatChannel, const [0,1]}
          Scale.X {type=MovieSceneScriptingFloatChannel, const [1,1]}
          Scale.Y {type=MovieSceneScriptingFloatChannel, const [1,1]}
          Shear.X {type=MovieSceneScriptingFloatChannel, const [0,1]}
          Shear.Y {type=MovieSceneScriptingFloatChannel, const [0,1]}
         ]
        }]
      }
     ]
    }]}
 }
 SlideFadeOut {
  binding_count=1
  length_seconds=0.25
  object=&self:SlideFadeOut
  sequence {bindings [RetainerBox {
     id=#2
     display_name=SizeBox_0
     possessed_class=RetainerBox
     tracks [
      {
       type=MovieSceneFloatTrack
       display_name=Render Opacity
       sections [{type=MovieSceneFloatSection, end_frame=5, start_frame=0, channels [None {type=MovieSceneScriptingFloatChannel, keys [[0,1],[5,0]]}]}]
      }
      {
       type=MovieScene2DTransformTrack
       sections [{
         type=MovieScene2DTransformSection
         end_frame=5
         start_frame=0
         channels [
          Translation.X {type=MovieSceneScriptingFloatChannel, keys [[0,0],[5,-150]]}
          Translation.Y {type=MovieSceneScriptingFloatChannel, const [0,1]}
          Angle {type=MovieSceneScriptingFloatChannel, const [0,1]}
          Scale.X {type=MovieSceneScriptingFloatChannel, const [1,1]}
          Scale.Y {type=MovieSceneScriptingFloatChannel, const [1,1]}
          Shear.X {type=MovieSceneScriptingFloatChannel, const [0,1]}
          Shear.Y {type=MovieSceneScriptingFloatChannel, const [0,1]}
         ]
        }]
      }
     ]
    }]}
 }
 SlideUp {
  binding_count=1
  length_seconds=0.25
  object=&self:SlideUp
  sequence {bindings [RetainerBox {
     id=#3
     display_name=SizeBox_0
     possessed_class=RetainerBox
     tracks [{
       type=MovieScene2DTransformTrack
       sections [{
         type=MovieScene2DTransformSection
         end_frame=5
         start_frame=0
         channels [
          Translation.X {type=MovieSceneScriptingFloatChannel, const [0,1]}
          Translation.Y {type=MovieSceneScriptingFloatChannel, keys [[0,0],[5,-55]]}
          Angle {type=MovieSceneScriptingFloatChannel, const [0,1]}
          Scale.X {type=MovieSceneScriptingFloatChannel, const [1,1]}
          Scale.Y {type=MovieSceneScriptingFloatChannel, const [1,1]}
          Shear.X {type=MovieSceneScriptingFloatChannel, const [0,1]}
          Shear.Y {type=MovieSceneScriptingFloatChannel, const [0,1]}
         ]
        }]
      }]
    }]}
 }
]
graphs [EventGraph {links [], nodes []}]
variables []
widgets {
 info {
  inheritedWidgetCount=4
  parentClass=/Script/FortniteUI.AthenaBroadcastKillFeedItemBase
  rootWidgetClass=&HorizontalBox
  widgetCount=15
 }
 widgets [
  HorizontalBox_0 {widgetClassPath=&HorizontalBox, widget=HorizontalBox_0}
  InstigatorBorder {
   widgetClassPath=&CommonBorder
   parent=HorizontalBox_0
   widget=InstigatorBorder
   properties {
    verticalAlignment=VAlign_Center
    background {
     bIsDynamicallyLoaded=false
     drawAs=Box
     imageType=NoImage
     mirroring=NoMirror
     resourceName=None
     resourceObject=null
     tiling=NoTile
     imageSize=<32 32>
     margin=<ltrb 0 0 0 0>
     outlineSettings {
      bUseBrushTransparency=false
      roundingType=HalfHeightRadius
      width=0
      color {colorUseRule=UseColor_Specified, specifiedColor=<rgba 0 0 0 0>}
      cornerRadii=<0 0 0 0>
     }
     tintColor {colorUseRule=UseColor_Specified, specifiedColor=<rgba 1 1 1 1>}
     uVRegion {bIsValid=false, max=<0 0>, min=<0 0>}
    }
    brushColor=<rgba 0 0 0 0.3>
    padding=<ltrb 17 0 12 0>
   }
   slot {class=HorizontalBoxSlot}
  }
  InstigatorHB {
   widgetClassPath=&HorizontalBox
   parent=InstigatorBorder
   widget=InstigatorHB
   slot {class=BorderSlot, properties {verticalAlignment=VAlign_Center, padding=<ltrb 17 0 12 0>}}
  }
  Text_InstigatorName {
   widgetClassPath=&CommonTextBlock
   bInherited=true
   parent=InstigatorHB
   widget=Text_InstigatorName
   properties {
    text=Instigator Name
    font {
     bForceMonospaced=false
     bMaterialIsStencil=false
     fontMaterial=null
     fontObject=&BurbankSmall
     letterSpacing=0
     monospacedWidth=1
     size=16
     skewAmount=0
     typefaceFontName=Bold
     outlineSettings {
      bApplyOutlineToDropShadows=false
      bMiteredCorners=false
      bSeparateFillAlpha=false
      outlineMaterial=null
      outlineSize=0
      outlineColor=<rgba 0 0 0 1>
     }
    }
    shadowColorAndOpacity=<rgba 0 0 0 0.5>
    shadowOffset=<0 0>
   }
   slot {class=HorizontalBoxSlot, properties {verticalAlignment=VAlign_Center}}
  }
  InstigatorElimsHB {
   widgetClassPath=&HorizontalBox
   parent=InstigatorHB
   widget=InstigatorElimsHB
   properties {visibility=Collapsed}
   slot {
    class=HorizontalBoxSlot
    properties {horizontalAlignment=HAlign_Left, verticalAlignment=VAlign_Center, padding=<ltrb 10 0 0 0>, size {sizeRule=Fill, value=1}}
   }
  }
  InstigatorElimsImage {
   widgetClassPath=&Image
   parent=InstigatorElimsHB
   widget=InstigatorElimsImage
   properties {
    visibility=SelfHitTestInvisible
    brush {
     bIsDynamicallyLoaded=false
     drawAs=Image
     imageType=NoImage
     mirroring=NoMirror
     resourceName=None
     resourceObject=/Game/Athena/HUD/Spectator/Art/T-Icon-BroadcastEliminations
     tiling=NoTile
     imageSize=<35 35>
     margin=<ltrb 0 0 0 0>
     outlineSettings {
      bUseBrushTransparency=false
      roundingType=HalfHeightRadius
      width=0
      color {colorUseRule=UseColor_Specified, specifiedColor=<rgba 0 0 0 0>}
      cornerRadii=<0 0 0 0>
     }
     tintColor {colorUseRule=UseColor_Specified, specifiedColor=<rgba 1 1 1 1>}
     uVRegion {bIsValid=false, max=<0 0>, min=<0 0>}
    }
    renderTransform {angle=0, scale=<1 1>, shear=<0 0>, translation=<-0.8571 -0.6984>}
   }
   slot {class=HorizontalBoxSlot, properties {horizontalAlignment=HAlign_Left, verticalAlignment=VAlign_Center, padding=<ltrb 0 0.5 1 0>}}
  }
  Image_EntryIcon {
   widgetClassPath=&Image
   bInherited=true
   parent=HorizontalBox_0
   widget=Image_EntryIcon
   properties {
    visibility=SelfHitTestInvisible
    brush {
     bIsDynamicallyLoaded=false
     drawAs=Image
     imageType=NoImage
     mirroring=NoMirror
     resourceName=None
     resourceObject=/Game/Athena/HUD/Spectator/Art/M_KillFeed_Explosion
     tiling=NoTile
     imageSize=<45 45>
     margin=<ltrb 0 0 0 0>
     outlineSettings {
      bUseBrushTransparency=false
      roundingType=HalfHeightRadius
      width=0
      color {colorUseRule=UseColor_Specified, specifiedColor=<rgba 0 0 0 0>}
      cornerRadii=<0 0 0 0>
     }
     tintColor {colorUseRule=UseColor_Specified, specifiedColor=<rgba 1 1 1 1>}
     uVRegion {bIsValid=false, max=<0 0>, min=<0 0>}
    }
   }
   slot {class=HorizontalBoxSlot, properties {horizontalAlignment=HAlign_Center, verticalAlignment=VAlign_Center, padding=<ltrb 10 0 10 0>}}
  }
  VictimOverlay {widgetClassPath=&Overlay, parent=HorizontalBox_0, widget=VictimOverlay, slot {class=HorizontalBoxSlot}}
  VictimBackground {
   widgetClassPath=/Game/UI/Foundation/Stats/BasicGradientFill
   parent=VictimOverlay
   widget=VictimBackground
   properties {rotation=0.25, color 1=<rgba 1 0.117 0.0252 1>, color 2=<rgba 1 0 0.491 1>, colorAndOpacity=<rgba 1 1 1 0.33>}
   slot {class=OverlaySlot, properties {horizontalAlignment=HAlign_Fill, verticalAlignment=VAlign_Fill}}
  }
  VictimBackgroundBlur {
   widgetClassPath=&BackgroundBlur
   parent=VictimOverlay
   widget=VictimBackgroundBlur
   properties {blurStrength=0.5}
   slot {class=OverlaySlot, properties {horizontalAlignment=HAlign_Fill, verticalAlignment=VAlign_Fill}}
  }
  VictimHB {
   widgetClassPath=&HorizontalBox
   parent=VictimOverlay
   widget=VictimHB
   slot {class=OverlaySlot, properties {horizontalAlignment=HAlign_Fill, verticalAlignment=VAlign_Center, padding=<ltrb 15 0 10 0>}}
  }
  Text_VictimName {
   widgetClassPath=&CommonTextBlock
   bInherited=true
   parent=VictimHB
   widget=Text_VictimName
   properties {
    text=Victim Name
    font {
     bForceMonospaced=false
     bMaterialIsStencil=false
     fontMaterial=null
     fontObject=&BurbankSmall
     letterSpacing=0
     monospacedWidth=1
     size=16
     skewAmount=0
     typefaceFontName=Bold
     outlineSettings {
      bApplyOutlineToDropShadows=false
      bMiteredCorners=false
      bSeparateFillAlpha=false
      outlineMaterial=null
      outlineSize=0
      outlineColor=<rgba 0 0 0 1>
     }
    }
    shadowOffset=<0 0>
   }
   slot {class=HorizontalBoxSlot, properties {verticalAlignment=VAlign_Center}}
  }
  VictimElimsHB {
   widgetClassPath=&HorizontalBox
   parent=VictimHB
   widget=VictimElimsHB
   properties {visibility=Collapsed}
   slot {class=HorizontalBoxSlot, properties {horizontalAlignment=HAlign_Left, verticalAlignment=VAlign_Center, padding=<ltrb 10 0 0 0>}}
  }
  VictimElimsImage {
   widgetClassPath=&Image
   parent=VictimElimsHB
   widget=VictimElimsImage
   properties {
    visibility=SelfHitTestInvisible
    brush {
     bIsDynamicallyLoaded=false
     drawAs=Image
     imageType=NoImage
     mirroring=NoMirror
     resourceName=None
     resourceObject=&T-Icon-BroadcastEliminations
     tiling=NoTile
     imageSize=<35 35>
     margin=<ltrb 0 0 0 0>
     outlineSettings {
      bUseBrushTransparency=false
      roundingType=HalfHeightRadius
      width=0
      color {colorUseRule=UseColor_Specified, specifiedColor=<rgba 0 0 0 0>}
      cornerRadii=<0 0 0 0>
     }
     tintColor {colorUseRule=UseColor_Specified, specifiedColor=<rgba 1 1 1 1>}
     uVRegion {bIsValid=false, max=<0 0>, min=<0 0>}
    }
    renderTransform {angle=0, scale=<1 1>, shear=<0 0>, translation=<-0.8571 -0.6984>}
   }
   slot {class=HorizontalBoxSlot, properties {horizontalAlignment=HAlign_Left, verticalAlignment=VAlign_Center, padding=<ltrb 0 0.5 1 0>}}
  }
  TutorialHighlight {widgetClassPath=&TutorialHighlight}
 ]
}
